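# Physics asset for the face accessory mesh of the Impulse HIS 03 male medium player cosmetic skin.
# It holds a single body on the head bone, shaped as one tilted capsule offset from the bone origin.
# It has no constraints, so there are no joints linking it to other bodies.
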PhysicsAsset /BRCosmetics/Characters/Player/Male/Medium/Bodies/M_MED_Impulse/Skins/Impulse_HIS_03/Meshes/M_MED_Impulse_HIS_3_FaceAcc_PhysicsAsset
bodies {head [{shapeType=Capsule, length=7.3234, radius=6.7434, center=<11.518 1.1525 -0.094>, rotation=<rot 0 -13.3532 -74.0004>}]}
constraints []
